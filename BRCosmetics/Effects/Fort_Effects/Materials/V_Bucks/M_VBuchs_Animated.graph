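# Lit mesh-particle material for animated V-Bucks bills, with per-particle color variation and a flexing motion.
# A hue-shifted gold color tints the bill texture, driven by dynamic PickColor, with vertex red masking and an optional pile variant.
# Rotation about local Y, scaled by U and sine of dynamic Time, bends the mesh and corrects normals.
# A PreviousFrameSwitch recomputes the bend at an earlier time to supply motion vectors.
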Material /BRCosmetics/Effects/Fort_Effects/Materials/V_Bucks/M_VBuchs_Animated
graph {
 properties {bUsedWithMeshParticles=true, pixelDepthOffsetMode=PDOM_Legacy, usedShadingModels=Default Lit}
 nodes [
  Constant3Vector()
  n1 = TextureSample() {texture=/Game/Effects/Fort_Effects/Textures/Emotes/V_Buck_Bill_v2}
  n2 = DynamicParameter() {paramNames [Time,PickColor,ParticleTime,Param4]}
  n3 = ComponentMask(None=TextureCoordinate()) {g=true}
  n4 = OneMinus(None=VertexColor().R)
  n5 = MaterialFunctionCall(
   Hue Shift Percentage=StaticSwitchParameter(True=LinearInterpolate(A=@n2.PickColor, B=@n3, Alpha=@n4), False=@n2.PickColor) {parameterName=Use on Pile}
   Texture=Constant3Vector() {constant=<rgba 1 0.8554 0.506 0>}
  ) {materialFunction=&HueShift}
  LinearInterpolate(A=Multiply(A=@n1.RGB, B=@n5.Result), B=@n5.Result, Alpha=OneMinus(None=VertexColor().R))
  LinearInterpolate(Alpha=@n1.R) {constA=0.4, constB=0.8}
  LinearInterpolate(Alpha=@n1.R) {constB=0.3}
  n6 = Transform(None=Constant3Vector() {constant=<rgba 0 1 0 0>}) {transformSourceType=TRANSFORMSOURCE_Local}
  n7 = Multiply(A=ComponentMask(None=TextureCoordinate()) {r=true}) {constB=0.25}
  n8 = DynamicParameter() {paramNames [Time,PickColor,ParticleTime,Param4]}
  n9 = Multiply(A=@n7, B=Sine(None=@n8.Time))
  MaterialFunctionCall(Rotation Axis=@n6, Rotation Angle=@n9) {materialFunction=&FixRotateAboutAxisNormals}
  ObjectPositionWS()
  ParticlePositionWS()
  n10 = TransformPosition(None=Constant3Vector()) {transformType=TRANSFORMPOSSOURCE_World}
  n11 = WorldPosition() {worldPositionShaderOffset=WPT_ExcludeAllShaderOffsets}
  n12 = Sine(None=Subtract(A=@n8.Time) {constB=0.1})
  PreviousFrameSwitch(
   Current Frame=RotateAboutAxis(NormalizedRotationAxis=@n6, RotationAngle=@n9, PivotPoint=@n10, Position=@n11.XYZ)
   Previous Frame=RotateAboutAxis(NormalizedRotationAxis=@n6, RotationAngle=Multiply(A=@n7, B=@n12), PivotPoint=@n10, Position=@n11.XYZ)
  )
  RotateAboutAxis(
   NormalizedRotationAxis=Transform(None=Constant3Vector() {constant=<rgba 0 1 0 0>})
   RotationAngle=MaterialFunctionCall() {materialFunction=&DebugTimeSine}.Result
   PivotPoint=TransformPosition(None=Constant3Vector()) {transformType=TRANSFORMPOSSOURCE_World}
   Position=WorldPosition().XYZ
  )
  ScalarParameter() {parameterName=ssssssss}
  ScalarParameter() {defaultValue=1, parameterName=Spec}
 ]
}
